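# Physics asset for the MilitaryFashion outfit of the medium female player character, covering only the right sleeve and right upper arm.
# A box body on the Dyn_sleeve_r bone and a capsule body on upperarm_r are joined by one constraint.
# The constraint locks both swing axes and twist, so it allows no angular movement between sleeve and arm.
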PhysicsAsset /Game/Characters/Player/Female/Medium/Bodies/F_MED_MilitaryFashion/Meshes/F_MED_Military_Fashion_Physics2
bodies {
 Dyn_sleeve_r [{shapeType=Box, extentX=7.3024, extentY=11.5515, extentZ=12.1744, center=<-1.4169 0.3649 -1.3363>, rotation=<rot -64.9966 89.999 90.0006>}]
 upperarm_r [{shapeType=Capsule, length=27.4395, radius=5.4733, center=<-14.6997 -0.3805 0.3345>, rotation=<rot -0.5689 -87.5496 -88.4367>}]
}
constraints [{
  bone1Name=Dyn_sleeve_r
  bone2Name=upperarm_r
  swing1LimitDegrees=5
  swing1Motion=Locked
  swing2LimitDegrees=5
  swing2Motion=Locked
  twistLimitDegrees=5
  twistMotion=Locked
 }]
