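# Physics asset for the medium female player Military Fashion body, covering only the left sleeve and upper arm.
# A box body on the dynamic bone Dyn_sleeve_l and a capsule body on upperarm_l are joined by one constraint with tight, equal swing and twist limit angles.
# The bone naming suggests secondary sleeve motion relative to the arm; the right side has no bodies.
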PhysicsAsset /Game/Characters/Player/Female/Medium/Bodies/F_MED_MilitaryFashion/Meshes/F_MED_Military_Fashion_PhysicsAsset
bodies {
 Dyn_sleeve_l [{shapeType=Box, extentX=3.8093, extentY=11.7103, extentZ=11.8849, center=<2.0381 -0.8411 0.6149>}]
 upperarm_l [{shapeType=Capsule, length=28.4512, radius=6.3182, center=<11.6066 -1.2706 -0.1386>, rotation=<rot -1.7997 93.531 -85.0474>}]
}
constraints [{bone1Name=Dyn_sleeve_l, bone2Name=upperarm_l, swing1LimitDegrees=5, swing2LimitDegrees=5, twistLimitDegrees=5}]
