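# Anim Blueprint for the female medium MilitaryFashion body cosmetic, making its skeleton follow the attached parent mesh.
# The output is only Copy Pose From Mesh, passed through local-to-component and back.
# Anim Dynamics sleeve nodes and two chained RigidBody nodes with separate physics assets exist but are disconnected from the output.
# The update event sets bUpdateOwnerAnimInputProperty to true, and an unused TryGetPawnOwner node remains.
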AnimBlueprint /BRCosmetics/Characters/Player/Female/Medium/Bodies/F_MED_MilitaryFashion/Meshes/F_MED_Military_Fashion_Skeleton_AnimBlueprint
target_skeleton=&/F_MED_Military_Fashion_Skeleton
graphs [
 AnimGraph {nodes [
   n1 = Anim Dynamics - Bone: Dyn_sleeve_r(Alpha=0.5, ExternalForce="0,0,0") {
    angularBiasOverride=2
    angularDampingOverride=10
    bGravityOverrideInSimSpace=true
    bOverrideLinearDamping=true
    bUsePlanarLimit=false
    gravityScale=0.5
    simulationSpace=World
    alphaBoolBlend {blendOption=Linear}
    boundBone {boneName=Dyn_sleeve_r}
    componentAppliedLinearAccClamp=<1000 1000 1000>
    componentLinearAccScale=<2 0.2 0.2>
    componentLinearVelScale=<0.2 0.2 0.2>
    physicsBodyDefinitions [{
      sphereCollisionRadius=0
      boundBone {boneName=Dyn_sleeve_r}
      boxExtents=<5 5 5>
      constraintSetup {angularConstraintType=Cone, coneAngle=6, angularLimitsMax=<3 3 3>, angularLimitsMin=<-3 -3 -3>, angularTarget=<0 0 0>}
      localJointOffset=<0 -4 -4>
     }]
    relativeSpaceBone {boneName=spine_03}
   }
   Anim Dynamics - Bone: Dyn_sleeve_l(ComponentPose=@n1, Alpha=0.5, ExternalForce="0,0,0") {
    angularBiasOverride=2
    angularDampingOverride=10
    bGravityOverrideInSimSpace=true
    bOverrideLinearDamping=true
    bUsePlanarLimit=false
    gravityScale=0.5
    simulationSpace=World
    alphaBoolBlend {blendOption=Linear}
    boundBone {boneName=Dyn_sleeve_l}
    componentAppliedLinearAccClamp=<1000 1000 1000>
    componentLinearAccScale=<2 0.2 0.2>
    componentLinearVelScale=<0.2 0.2 0.2>
    physicsBodyDefinitions [{
      sphereCollisionRadius=0
      boundBone {boneName=Dyn_sleeve_l}
      boxExtents=<5 5 5>
      constraintSetup {angularConstraintType=Cone, coneAngle=6, angularLimitsMax=<3 3 3>, angularLimitsMin=<-3 -3 -3>, angularTarget=<0 0 0>}
      localJointOffset=<0 -4 -4>
     }]
    relativeSpaceBone {boneName=spine_03}
   }
   n2 = RigidBody() {
    bClampLinearTranslationLimitToRefPose=true
    bOverrideWorldGravity=true
    overridePhysicsAsset=/Game/Characters/Player/Female/Medium/Bodies/F_MED_MilitaryFashion/Meshes/F_MED_Military_Fashion_Physics2
    simulationSpace=WorldSpace
    alphaBoolBlend {blendOption=Linear}
    componentLinearAccScale=<1 1 1>
    componentLinearVelScale=<1 1 1>
    overrideWorldGravity {z=-980}
   }
   RigidBody(ComponentPose=@n2) {
    bClampLinearTranslationLimitToRefPose=true
    bOverrideWorldGravity=true
    overridePhysicsAsset=/Game/Characters/Player/Female/Medium/Bodies/F_MED_MilitaryFashion/Meshes/F_MED_Military_Fashion_PhysicsAsset
    simulationSpace=WorldSpace
    alphaBoolBlend {blendOption=Linear}
    componentLinearAccScale=<1 1 1>
    componentLinearVelScale=<1 1 1>
    overrideWorldGravity {z=-980}
   }
   n3 = Copy Pose From Mesh() {bUseAttachedParent=true}
   Output Pose(Result=Component To Local(ComponentPose=Local To Component(LocalPose=@n3))) {layerGroup=None}
  ]}
 EventGraph {nodes [
   Event BlueprintUpdateAnimation()
   > Set bUpdateOwnerAnimInputProperty(bUpdateOwnerAnimInputProperty:bool=true) -> (Output_Get:bool)
   TryGetPawnOwner()
  ]}
]
variables []
