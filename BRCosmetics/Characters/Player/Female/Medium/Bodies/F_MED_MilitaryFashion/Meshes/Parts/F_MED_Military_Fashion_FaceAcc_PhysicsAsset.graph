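# Physics asset for the face accessory part of the female medium Military Fashion cosmetic body, covering a hair bun on the head.
# Two capsules form the head body, a box forms the bun's master bone, and four spheres form the smaller bun bones.
# The small bones are each constrained to the master, which is constrained to the head.
# All joints share tight swing and twist limits, allowing only slight angular movement.
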PhysicsAsset /BRCosmetics/Characters/Player/Female/Medium/Bodies/F_MED_MilitaryFashion/Meshes/Parts/F_MED_Military_Fashion_FaceAcc_PhysicsAsset
bodies {
 dyn_hairBun_a [{shapeType=Sphere, radius=2.8783, center=<0.7432 -0.0001 0>}]
 dyn_hairBun_b [{shapeType=Sphere, radius=2.1933, center=<-2.0042 -0.7622 -0.4089>}]
 dyn_hairBun_c [{shapeType=Sphere, radius=1.8635, center=<1.6584 -1.0903 0.0269>}]
 dyn_hairBun_d [{shapeType=Sphere, radius=1.755, center=<2.5415 -0.7305 1.9806>}]
 dyn_hairBun_mstr [{shapeType=Box, extentX=3.8529, extentY=8.7566, extentZ=9.2666, rotation=<rot 30.0001 -90 0>}]
 head [
  {shapeType=Capsule, length=6.1875, radius=4.1675, center=<2.6251 5.4011 0.1871>, rotation=<rot -0.3859 90.9394 -89.3313>}
  {shapeType=Capsule, length=10.75, radius=4.25, center=<7.3484 4.0299 -0.0761>, rotation=<rot -2.5759 14.7849 94.6661>}
 ]
}
constraints [
 {bone1Name=dyn_hairBun_a, bone2Name=dyn_hairBun_mstr, swing1LimitDegrees=5, swing2LimitDegrees=5, twistLimitDegrees=5}
 {bone1Name=dyn_hairBun_b, bone2Name=dyn_hairBun_mstr, swing1LimitDegrees=5, swing2LimitDegrees=5, twistLimitDegrees=5}
 {bone1Name=dyn_hairBun_c, bone2Name=dyn_hairBun_mstr, swing1LimitDegrees=5, swing2LimitDegrees=5, twistLimitDegrees=5}
 {bone1Name=dyn_hairBun_d, bone2Name=dyn_hairBun_mstr, swing1LimitDegrees=5, swing2LimitDegrees=5, twistLimitDegrees=5}
 {bone1Name=dyn_hairBun_mstr, bone2Name=head, swing1LimitDegrees=5, swing2LimitDegrees=5, twistLimitDegrees=5}
]
