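# Physics asset for the female medium MilitaryFashion face accessory mesh, setting up dynamic hair and bang bones.
# Left and right three-bone capsule hair chains and single box bangs hang from the head with tight, asymmetric swing and twist limits.
# Capsules on spine, clavicles and arms, plus a head capsule and large offset box, form body shapes.
# Head, arms and upper spine link to spine_04 with free angular motion; clavicles have no constraints.
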PhysicsAsset /BRCosmetics/Characters/Player/Female/Medium/Bodies/F_MED_MilitaryFashion/Meshes/Parts/F_MED_Military_Fashion_FaceAcc_PhysicsAsset1
bodies {
 clavicle_l [{shapeType=Capsule, length=12.75, radius=5, center=<4.4789 0.5409 1.2582>, rotation=<rot 90 0 0>}]
 clavicle_r [{shapeType=Capsule, length=11.25, radius=5, rotation=<rot 90 0 0>}]
 dyn_bang_01_l [{shapeType=Box, extentX=6.3768, extentY=2.4519, extentZ=1.3097, center=<3.156 0 0>}]
 dyn_bang_01_r [{shapeType=Box, extentX=6.6251, extentY=2.6918, extentZ=0.9126, center=<-3.23 0 0>}]
 dyn_hair_01_l [{shapeType=Capsule, length=6.71, radius=1.8422, center=<5.0181 0 0>, rotation=<rot 90 0 0>}]
 dyn_hair_01_r [{shapeType=Capsule, length=6.838, radius=1.3668, center=<4.4174 0 0>, rotation=<rot 90 0 0>}]
 dyn_hair_02_l [{shapeType=Capsule, length=6.1438, radius=1.2234, center=<3.7803 0 -0.0001>, rotation=<rot 90 0 -6.2654>}]
 dyn_hair_02_r [{shapeType=Capsule, length=8.1832, radius=1.0937, center=<4.4477 0.3603 -0.3667>, rotation=<rot 90 0 0>}]
 dyn_hair_03_l [{shapeType=Capsule, length=6.5253, radius=0.9368, center=<3.4004 1.5842 -0.8139>, rotation=<rot 77.666 27.0044 0>}]
 dyn_hair_03_r [{shapeType=Capsule, length=8.4724, radius=1.078, center=<3.661 -1.6014 -1.4806>, rotation=<rot -0.0013 -112.8134 -68.636>}]
 head [
  {shapeType=Capsule, length=9.4342, radius=6.6682, center=<3.2615 2.0181 -0.0526>, rotation=<rot -0.271 -93.2675 -89.8435>}
  {shapeType=Box, extentX=35.5, extentY=6.5199, extentZ=45.5061, center=<0 -7.9943 -0.0002>}
 ]
 lowerarm_l [{shapeType=Capsule, length=25.5, radius=4.5, center=<10.4666 0 -0.0001>, rotation=<rot 90 0.0198 0.0199>}]
 lowerarm_r [{shapeType=Capsule, length=25.5, radius=4.5, center=<-9.5652 -2.1523 0.3466>, rotation=<rot 90 -0.0099 -0.0099>}]
 spine_04 [{shapeType=Capsule, length=14.25, radius=14.25, center=<-1.4441 0.2029 0>}]
 spine_05 [{shapeType=Capsule, length=14.25, radius=14.25, center=<-3.477 -3.3625 0>}]
 upperarm_l [{shapeType=Capsule, length=18.75, radius=4.5, center=<10.0596 -0.0001 0>, rotation=<rot 90 0.0099 0.0099>}]
 upperarm_r [{shapeType=Capsule, length=25.5, radius=4.5, center=<-12.8191 -0.0001 0>, rotation=<rot 90 179.9802 179.9803>}]
}
constraints [
 {bone1Name=dyn_hair_01_l, bone2Name=head, swing1LimitDegrees=20, swing2LimitDegrees=5, twistLimitDegrees=5}
 {bone1Name=dyn_hair_02_l, bone2Name=dyn_hair_01_l, swing1LimitDegrees=40, swing2LimitDegrees=5, twistLimitDegrees=5}
 {bone1Name=dyn_bang_01_r, bone2Name=head, swing1LimitDegrees=20, swing2LimitDegrees=15, twistLimitDegrees=5}
 {bone1Name=dyn_bang_01_l, bone2Name=head, swing1LimitDegrees=20, swing2LimitDegrees=15, twistLimitDegrees=5}
 {
  bone1Name=spine_05
  bone2Name=spine_04
  swing1LimitDegrees=45
  swing1Motion=Free
  swing2LimitDegrees=45
  swing2Motion=Free
  twistLimitDegrees=45
  twistMotion=Free
 }
 {
  bone1Name=upperarm_l
  bone2Name=spine_04
  swing1LimitDegrees=45
  swing1Motion=Free
  swing2LimitDegrees=45
  swing2Motion=Free
  twistLimitDegrees=45
  twistMotion=Free
 }
 {
  bone1Name=lowerarm_l
  bone2Name=spine_04
  swing1LimitDegrees=45
  swing1Motion=Free
  swing2LimitDegrees=45
  swing2Motion=Free
  twistLimitDegrees=45
  twistMotion=Free
 }
 {
  bone1Name=upperarm_r
  bone2Name=spine_04
  swing1LimitDegrees=45
  swing1Motion=Free
  swing2LimitDegrees=45
  swing2Motion=Free
  twistLimitDegrees=45
  twistMotion=Free
 }
 {
  bone1Name=lowerarm_r
  bone2Name=spine_04
  swing1LimitDegrees=45
  swing1Motion=Free
  swing2LimitDegrees=45
  swing2Motion=Free
  twistLimitDegrees=45
  twistMotion=Free
 }
 {
  bone1Name=head
  bone2Name=spine_04
  swing1LimitDegrees=45
  swing1Motion=Free
  swing2LimitDegrees=45
  swing2Motion=Free
  twistLimitDegrees=45
  twistMotion=Free
 }
 {bone1Name=dyn_hair_01_r, bone2Name=head, swing1LimitDegrees=10, swing2LimitDegrees=15, twistLimitDegrees=5}
 {bone1Name=dyn_hair_02_r, bone2Name=dyn_hair_01_r, swing1LimitDegrees=45, swing2LimitDegrees=25, twistLimitDegrees=5}
 {bone1Name=dyn_hair_03_r, bone2Name=dyn_hair_02_r, swing1LimitDegrees=25, swing2LimitDegrees=25, twistLimitDegrees=5}
 {bone1Name=dyn_hair_03_l, bone2Name=dyn_hair_02_l, swing1LimitDegrees=40, swing2LimitDegrees=10, twistLimitDegrees=5}
]
