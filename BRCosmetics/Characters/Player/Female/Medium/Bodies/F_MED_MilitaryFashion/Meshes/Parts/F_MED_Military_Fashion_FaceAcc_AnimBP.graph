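# Animation Blueprint for the Military Fashion female face-accessory mesh, adding physics motion to hair and bang bones on a pose copied from the parent mesh.
# Two chained RigidBody passes simulate relative to spine_05, with constraints between them tying dyn_bang_01_l to dyn_hair_01_l.
# Above a running-speed threshold, a third RigidBody pass blends in over the cached result.
# The event graph also computes BraidMoveIn and SpineRot, which the AnimGraph leaves unused.
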AnimBlueprint /BRCosmetics/Characters/Player/Female/Medium/Bodies/F_MED_MilitaryFashion/Meshes/Parts/F_MED_Military_Fashion_FaceAcc_AnimBP
target_skeleton=&/F_MED_Military_Fashion_FaceAcc_Skeleton
graphs [
 AnimGraph {nodes [
   n1 = RigidBody(ComponentPose=Local To Component(LocalPose=Use cached pose 'RigidBodies'())) {
    bClampLinearTranslationLimitToRefPose=true
    bFreezeIncomingPoseOnStart=true
    bOverrideWorldGravity=true
    bTransferBoneVelocities=true
    overridePhysicsAsset=&/F_MED_Military_Fashion_FaceAcc_PhysicsAsset
    simulationSpace=BaseBoneSpace
    alphaBoolBlend {blendOption=Linear}
    baseBoneRef {boneName=spine_05}
    componentLinearAccScale=<0.1 0.1 1>
    componentLinearVelScale=<0.1 0.1 0.1>
    overrideWorldGravity {z=-980}
   }
   Output Pose(Result=Blend Poses by bool(
     BlendPose_0=Component To Local(ComponentPose=@n1)
     BlendPose_1=Use cached pose 'RigidBodies'()
     bActiveValue=Get IsRunning() -> (IsRunning:bool)
    ) {blendType=Linear}) {layerGroup=None}
   Save cached pose 'SubGraph'(Pose=Copy Pose From Mesh(SourceMeshComponent=Get MeshToCopy() -> (MeshToCopy:Skeletal Mesh Component Object Reference)) {bCopyCurves=true, bUseAttachedParent=true}) {cacheName=SubGraph}
   n2 = RigidBody(ComponentPose=Local To Component(LocalPose=Use cached pose 'SubGraph'())) {
    bClampLinearTranslationLimitToRefPose=true
    bForceDisableCollisionBetweenConstraintBodies=true
    bOverrideWorldGravity=true
    overridePhysicsAsset=&/F_MED_Military_Fashion_FaceAcc_PhysicsAsset1
    simulationSpace=BaseBoneSpace
    alphaBoolBlend {blendOption=Linear}
    baseBoneRef {boneName=spine_05}
    componentAppliedLinearAccClamp=<3000 3000 3000>
    componentLinearAccScale=<0.354 0.3 0.3>
    componentLinearVelScale=<0.3 0.3 0.3>
    overrideWorldGravity {z=-980}
   }
   n3 = "Constraint - dyn_bang_01_l (R)"(ComponentPose=@n2, Alpha=1, ConstraintWeights_0=0.5) {
    alphaBoolBlend {blendOption=Linear}
    boneToModify {boneName=dyn_bang_01_l}
    constraintSetup [{transformType=Rotation, targetBone {boneName=dyn_hair_01_l}}]
    constraintWeights [1]
   }
   n4 = "Constraint - dyn_bang_01_l (P)"(ComponentPose=@n3, Alpha=1, ConstraintWeights_0=1) {
    alphaBoolBlend {blendOption=Linear}
    boneToModify {boneName=dyn_bang_01_l}
    constraintSetup [{perAxis {bX=false, bZ=false}, targetBone {boneName=dyn_hair_01_l}}]
    constraintWeights [1]
   }
   n5 = RigidBody(ComponentPose=@n4) {
    bClampLinearTranslationLimitToRefPose=true
    bForceDisableCollisionBetweenConstraintBodies=true
    bOverrideWorldGravity=true
    overridePhysicsAsset=&F_MED_Military_Fashion_FaceAcc_PhysicsAsset
    simulationSpace=BaseBoneSpace
    alphaBoolBlend {blendOption=Linear}
    baseBoneRef {boneName=spine_05}
    componentLinearAccScale=<0.5 0.5 0.5>
    componentLinearVelScale=<0.5 0.5 0.5>
    overrideWorldGravity {z=-980}
   }
   Save cached pose 'RigidBodies'(Pose=Component To Local(ComponentPose=@n5)) {cacheName=RigidBodies}
  ]}
 EventGraph {nodes [
   n1 = Get AimPitch()
   PrintString!(execute, InString="To String (Float)"(InDouble=@n1), bPrintToScreen, bPrintToLog, TextColor, Duration=0.03, Key) -> (then)
   > PrintString(InString="To String (Float)"(InDouble=Get AimYaw() -> (AimYaw:float)), Duration=0.03)
   Event BlueprintUpdateAnimation()
   > Set bUpdateOwnerAnimInputProperty(bUpdateOwnerAnimInputProperty:bool=true) -> (Output_Get:bool)
   > Set BraidMoveIn(BraidMoveIn:double=MapRangeClamped(Value="Absolute (Float)"(A=@n1), InRangeA=0, InRangeB=90, OutRangeB=45)) -> (Output_Get:double)
   n2 = GetOwnerSkeletalMeshComponent()
   > Set MeshToCopy(MeshToCopy:Skeletal Mesh Component Object Reference=@n2) -> (Output_Get:Skeletal Mesh Component Object Reference)
   > Set IsRunning(IsRunning:bool="float > float"(A=Get PawnSpeed2D(), B=420) {is=title K2Node_CallFunction}) -> (Output_Get:bool)
   > Set SpineRot(SpineRot:Rotator=GetSocketRotation!(self:Scene Component Object Reference=@n2, InSocketName:Name=spine_04) -> (ReturnValue:Rotator)) -> (Output_Get:Rotator)
   TryGetPawnOwner()
  ]}
]
variables [
 BraidMoveIn {type=number}
 IsRunning {type=boolean}
 MeshToCopy {type=SkeletalMeshComponent}
 SpineRot {type=Rotator}
]
